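# Sound cue for the Alchemy pickaxe cosmetic's impact sound when it hits an enemy.
# A Modulator feeds an equally weighted Random node that picks one of four enemy-hit wave variations.
# It is primed on load, routed through the impacts sound class, and uses shared pickaxe impact attenuation and property sheet.
# A player melee weapon concurrency set governs simultaneous playback.
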SoundCue /BRCosmetics/Sounds/PickAxes/Alchemy/PickAxe_Alchemy_Enemy_Impact_Cue
root=Modulator_0
nodes [
 {id=Modulator_0, children [Random_3]}
 {id=Random_3, children [WavePlayer_36, WavePlayer_37, WavePlayer_38, WavePlayer_39], properties {weights [1,1,1,1]}}
 {id=WavePlayer_36, properties {soundWaveAssetPtr=&/PickAxe_Alchemy_Hit_Enemy_01}}
 {id=WavePlayer_37, properties {soundWaveAssetPtr=&/PickAxe_Alchemy_Hit_Enemy_02}}
 {id=WavePlayer_38, properties {soundWaveAssetPtr=&/PickAxe_Alchemy_Hit_Enemy_03}}
 {id=WavePlayer_39, properties {soundWaveAssetPtr=&/PickAxe_Alchemy_Hit_Enemy_04}}
]
properties {
 attenuationSettings=/Game/Sounds/Attenuation/Fort_Pickaxe_Impact_Att
 audioPropertiesSheet=/BRCosmetics/Sounds/Settings/PropertySheets/Pickaxe/APS_Cosmetic_Pickaxe_Impact
 bPrimeOnLoad=true
 maxDistance=4300
 soundClassObject=&Fort_Impacts
 volumeMultiplier=0.9375
 attenuationOverrides {
  bAttenuateWithLPF=true
  distanceAlgorithm=Logarithmic
  falloffDistance=800
  lPFRadiusMax=800
  lPFRadiusMin=400
  attenuationShapeExtents {x=300}
 }
 concurrencyOverrides {bLimitToOwner=true, maxCount=2, resolutionRule=StopOldest}
 concurrencySet [&Player_Melee_Weapon_Conc]
}
