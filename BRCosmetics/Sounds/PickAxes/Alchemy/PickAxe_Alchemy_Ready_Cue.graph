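# Sound cue for the Alchemy pickaxe's ready sound, mixing close and medium attenuation layers.
# The close layer pairs pitched-down random Alchemy swings with a delayed hand grab, while the medium layer plays random distant fire axe swings.
# Quality switches drop the grab and use smaller swing pools in the last two slots.
# A concurrency override limits instances per owner, stopping the oldest.
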SoundCue /BRCosmetics/Sounds/PickAxes/Alchemy/PickAxe_Alchemy_Ready_Cue
root=Mixer_2
nodes [
 {id=Attenuation_2, children [Mixer_0], properties {attenuationSettings=/Game/Sounds/Attenuation/Fort_Weapon_Pickaxe_Close_Att}}
 {id=Attenuation_3, children [QualityLevel_0], properties {attenuationSettings=/Game/Sounds/Attenuation/Fort_Weapon_Pickaxe_Medium_Att}}
 {id=Delay_0, children [WavePlayer_1], properties {delayMax=0.25, delayMin=0.25}}
 {id=Mixer_0, children [QualityLevel_1, Modulator_0], properties {inputVolume [3,1]}}
 {id=Mixer_2, children [Attenuation_2, Attenuation_3], properties {inputVolume [0.5,1]}}
 {id=Modulator_0, children [QualityLevel_3], properties {pitchMax=0.8, pitchMin=0.8}}
 {id=Modulator_3, children [Random_5]}
 {id=QualityLevel_0, children [Random_2,Random_2,Random_3,Random_3]}
 {id=QualityLevel_1, children [Delay_0,Delay_0,null,null]}
 {id=QualityLevel_3, children [Random_7,Random_7,Modulator_3,Modulator_3]}
 {
  id=Random_2
  children [WavePlayer_0, WavePlayer_36, WavePlayer_37, WavePlayer_38, WavePlayer_39, WavePlayer_40]
  properties {weights [1,1,1,1,1,1]}
 }
 {id=Random_3, children [WavePlayer_42, WavePlayer_41, WavePlayer_43], properties {bShouldExcludeFromBranchCulling=true, weights [1,1,1]}}
 {id=Random_5, children [WavePlayer_70, WavePlayer_71, WavePlayer_72], properties {bShouldExcludeFromBranchCulling=true, weights [1,1,1]}}
 {
  id=Random_7
  children [WavePlayer_64, WavePlayer_65, WavePlayer_66, WavePlayer_67, WavePlayer_68, WavePlayer_69]
  properties {weights [1,1,1,1,1,1]}
 }
 {id=WavePlayer_0, properties {soundWaveAssetPtr=/Game/Athena/Sounds/Weapons/Fire_Axe_Swing_Distant_04}}
 {id=WavePlayer_1, properties {soundWaveAssetPtr=&/PickAxe_Alchemy_Hand_Grab_01}}
 {id=WavePlayer_36, properties {soundWaveAssetPtr=/Game/Athena/Sounds/Weapons/Fire_Axe_Swing_Distant_01}}
 {id=WavePlayer_37, properties {soundWaveAssetPtr=/Game/Athena/Sounds/Weapons/Fire_Axe_Swing_Distant_02}}
 {id=WavePlayer_38, properties {soundWaveAssetPtr=/Game/Athena/Sounds/Weapons/Fire_Axe_Swing_Distant_03}}
 {id=WavePlayer_39, properties {soundWaveAssetPtr=/Game/Athena/Sounds/Weapons/Fire_Axe_Swing_Distant_05}}
 {id=WavePlayer_40, properties {soundWaveAssetPtr=/Game/Athena/Sounds/Weapons/Fire_Axe_Swing_Distant_06}}
 {id=WavePlayer_41, properties {soundWaveAssetPtr=&Fire_Axe_Swing_Distant_03}}
 {id=WavePlayer_42, properties {soundWaveAssetPtr=&Fire_Axe_Swing_Distant_04}}
 {id=WavePlayer_43, properties {soundWaveAssetPtr=&Fire_Axe_Swing_Distant_06}}
 {id=WavePlayer_64, properties {soundWaveAssetPtr=&/PickAxe_Alchemy_Swing_01}}
 {id=WavePlayer_65, properties {soundWaveAssetPtr=&/PickAxe_Alchemy_Swing_02}}
 {id=WavePlayer_66, properties {soundWaveAssetPtr=&/PickAxe_Alchemy_Swing_03}}
 {id=WavePlayer_67, properties {soundWaveAssetPtr=&/PickAxe_Alchemy_Swing_04}}
 {id=WavePlayer_68, properties {soundWaveAssetPtr=&/PickAxe_Alchemy_Swing_05}}
 {id=WavePlayer_69, properties {soundWaveAssetPtr=&/PickAxe_Alchemy_Swing_06}}
 {id=WavePlayer_70, properties {soundWaveAssetPtr=&PickAxe_Alchemy_Swing_01}}
 {id=WavePlayer_71, properties {soundWaveAssetPtr=&PickAxe_Alchemy_Swing_02}}
 {id=WavePlayer_72, properties {soundWaveAssetPtr=&PickAxe_Alchemy_Swing_03}}
]
properties {
 audioPropertiesSheet=/BRCosmetics/Sounds/Settings/PropertySheets/Pickaxe/APS_Cosmetic_Pickaxe_Ready
 bOverrideConcurrency=true
 bPrimeOnLoad=true
 maxDistance=3000
 soundClassObject=&Fort_WeaponMelee_PSM
 volumeMultiplier=0.2
 concurrencyOverrides {bLimitToOwner=true, maxCount=2, resolutionRule=StopOldest}
 concurrencySet [&Player_Melee_Weapon_Conc]
}
