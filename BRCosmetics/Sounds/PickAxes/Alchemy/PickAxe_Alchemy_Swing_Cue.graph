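# Sound cue for the Alchemy cosmetic pickaxe swing, mixing a close Alchemy layer with a medium-attenuation layer of shared distant fire axe swings.
# Each layer passes through a quality switch that picks randomly from six waves on the first two levels or a three-wave subset on the others.
# The close layer is mixed at half volume, and its reduced branch adds a modulator.
# An owner-limited concurrency override stops the oldest swing.
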SoundCue /BRCosmetics/Sounds/PickAxes/Alchemy/PickAxe_Alchemy_Swing_Cue
root=Mixer_0
nodes [
 {id=Attenuation_0, children [QualityLevel_2], properties {attenuationSettings=/Game/Sounds/Attenuation/Fort_Weapon_Pickaxe_Close_Att}}
 {id=Attenuation_1, children [QualityLevel_1], properties {attenuationSettings=/Game/Sounds/Attenuation/Fort_Weapon_Pickaxe_Medium_Att}}
 {id=Mixer_0, children [Attenuation_0, Attenuation_1], properties {inputVolume [0.5,1]}}
 {id=Modulator_2, children [Random_4]}
 {id=QualityLevel_1, children [Random_0,Random_0,Random_1,Random_1]}
 {id=QualityLevel_2, children [Random_7,Random_7,Modulator_2,Modulator_2]}
 {id=Random_0, children [WavePlayer_1,WavePlayer_2,WavePlayer_4,WavePlayer_6,WavePlayer_7,WavePlayer_8], properties {weights [1,1,1,1,1,1]}}
 {id=Random_1, children [WavePlayer_10, WavePlayer_9, WavePlayer_11], properties {bShouldExcludeFromBranchCulling=true, weights [1,1,1]}}
 {id=Random_4, children [WavePlayer_70, WavePlayer_71, WavePlayer_72], properties {bShouldExcludeFromBranchCulling=true, weights [1,1,1]}}
 {
  id=Random_7
  children [WavePlayer_64, WavePlayer_65, WavePlayer_66, WavePlayer_67, WavePlayer_68, WavePlayer_69]
  properties {weights [1,1,1,1,1,1]}
 }
 {id=WavePlayer_1, properties {soundWaveAssetPtr=/Game/Athena/Sounds/Weapons/Fire_Axe_Swing_Distant_04}}
 {id=WavePlayer_10, properties {soundWaveAssetPtr=&Fire_Axe_Swing_Distant_04}}
 {id=WavePlayer_11, properties {soundWaveAssetPtr=/Game/Athena/Sounds/Weapons/Fire_Axe_Swing_Distant_06}}
 {id=WavePlayer_2, properties {soundWaveAssetPtr=/Game/Athena/Sounds/Weapons/Fire_Axe_Swing_Distant_01}}
 {id=WavePlayer_4, properties {soundWaveAssetPtr=/Game/Athena/Sounds/Weapons/Fire_Axe_Swing_Distant_02}}
 {id=WavePlayer_6, properties {soundWaveAssetPtr=/Game/Athena/Sounds/Weapons/Fire_Axe_Swing_Distant_03}}
 {id=WavePlayer_64, properties {soundWaveAssetPtr=&/PickAxe_Alchemy_Swing_01}}
 {id=WavePlayer_65, properties {soundWaveAssetPtr=&/PickAxe_Alchemy_Swing_02}}
 {id=WavePlayer_66, properties {soundWaveAssetPtr=&/PickAxe_Alchemy_Swing_03}}
 {id=WavePlayer_67, properties {soundWaveAssetPtr=&/PickAxe_Alchemy_Swing_04}}
 {id=WavePlayer_68, properties {soundWaveAssetPtr=&/PickAxe_Alchemy_Swing_05}}
 {id=WavePlayer_69, properties {soundWaveAssetPtr=&/PickAxe_Alchemy_Swing_06}}
 {id=WavePlayer_7, properties {soundWaveAssetPtr=/Game/Athena/Sounds/Weapons/Fire_Axe_Swing_Distant_05}}
 {id=WavePlayer_70, properties {soundWaveAssetPtr=&PickAxe_Alchemy_Swing_01}}
 {id=WavePlayer_71, properties {soundWaveAssetPtr=&PickAxe_Alchemy_Swing_02}}
 {id=WavePlayer_72, properties {soundWaveAssetPtr=&PickAxe_Alchemy_Swing_03}}
 {id=WavePlayer_8, properties {soundWaveAssetPtr=&Fire_Axe_Swing_Distant_06}}
 {id=WavePlayer_9, properties {soundWaveAssetPtr=&Fire_Axe_Swing_Distant_03}}
]
properties {
 audioPropertiesSheet=/BRCosmetics/Sounds/Settings/PropertySheets/Pickaxe/APS_Cosmetic_Pickaxe_Swing
 bOverrideConcurrency=true
 bPrimeOnLoad=true
 maxDistance=3000
 soundClassObject=&Fort_WeaponMelee_PSM
 volumeMultiplier=0.45
 concurrencyOverrides {bLimitToOwner=true, maxCount=2, resolutionRule=StopOldest}
 concurrencySet [&Player_Melee_Weapon_Conc]
}
